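# Stereo music MetaSound source for the DragonCart orb shock sequence, playing the MX_S39_DragonCart_OrbShock_STNG stinger on the gameplay music sound class.
# OnPlay starts a stereo Wave Player whose left and right outputs are each multiplied by a shared float gain.
# That gain comes from a Value node that starts at unity and is set by the player's On Looped trigger.
# The source uses sequence-specific attenuation and keeps playing when silent.
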MetaSoundSource /DragonCartNarrative/Assets/Audio/OrbShockSequence/Music/MSS_MX_S39_DragonCart_OrbShockSeq
inputs [
 Sound {type=WaveAsset}
 UE.Source.OnPlay {type=Trigger}
 UE.Source.StartTime.StartTime {type=Time}
]
interfaces [UE.OutputFormat.Stereo:1.0, UE.Source.StartTime:1.0, UE.Source:1.1]
outputs [
 "UE.OutputFormat.Stereo.Audio:0" {type=Audio}
 "UE.OutputFormat.Stereo.Audio:1" {type=Audio}
]
pages [Default {
  input_defaults {
   Sound=&/MX_S39_DragonCart_OrbShock_STNG
   UE.Source.OnPlay=null
   UE.Source.StartTime.StartTime=null
  }
  nodes [
   InitVariable.Time()
   InitVariable.Trigger()
   WavePlayer_0 = UE.Wave Player:Stereo(Wave Asset=^Sound, Play=^UE.Source.OnPlay) {also {defaults {Loop="<MetaSoundBuilderResult.SUCCEEDED: 0>"}, version [1,1]}}
   n1 = Value.Value:Float(Set=@WavePlayer_0.On Looped) {also {defaults {Init Value=1}}}
   Multiply_0 = UE.Multiply:Audio by Float(AdditionalOperands=@n1.Output Value, PrimaryOperand=@WavePlayer_0.Out Left)
   Multiply_1 = UE.Multiply:Audio by Float(AdditionalOperands=@n1.Output Value, PrimaryOperand=@WavePlayer_0.Out Right)
  ]
  links [{dst="@UE.OutputFormat.Stereo.Audio:0", src=Multiply_0.Out}, {dst="@UE.OutputFormat.Stereo.Audio:1", src=Multiply_1.Out}]
 }]
properties {
 assetImportData=&self:AssetImportData
 attenuationSettings=&/ATT_DragonCart_ShockSeq_Music
 bLooping=true
 duration=10000
 maxDistance=44000
 numChannels=2
 outputFormat=Stereo
 sampleRate=48000
 soundClassObject=/Game/Sounds/FortSoundClassesAndModes/Fort_Music_Gameplay
 virtualizationMode=PlayWhenSilent
 volume=4.1
}
